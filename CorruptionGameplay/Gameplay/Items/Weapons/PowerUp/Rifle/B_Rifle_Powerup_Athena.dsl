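; Athena power-up rifle weapon with an overheat mechanic and a replicated "heat event" that temporarily overrides fire mode data.
; Server mirrors heat into a replicated value, whose RepNotify updates glow material, Niagara and audio parameters.
; Server-only start/end events toggle a replicated bool, whose RepNotify switches fire mode, glow and loop audio.
; Overheating ends the heat event and triggers cooldown effects, reset upon cooling, while shots add heat-scaled muzzle FX and sound.
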(event EventBeginPlay
  (|CallParentFunction)
  (Networking|SwitchHasAuthority
    (:Authority
      ;[1 1 1] Replicates the heat value and calls FX scripting on rep notify.
      (Overheating|AssignOnOverheatValueChanged (delegate Custom|UpdateCurrentHeatValue)))))

(event Custom|UpdateCurrentHeatValue
  (params (CurrentValue :type Float%20%28single-precision%29) (MaxValue :type Float%20%28single-precision%29) (State :type EFortWeaponOverheatState%20Enum) (bIsInCooldown :type Boolean))
  ;[1 1 1] Replicates the heat value and calls FX scripting on rep notify.
  (Variables|Default|SetCurrentHeatValueRep CurrentValue))

(event Custom|OnStartHeatEvent_Rep
  ;[1 1 1] Heat Event Logic
  ;;[1 1 1] On Start Heat Event
  (Networking|CanExecuteCosmeticEvents
    (:True
      (FortWeapon|SetOverrideFireModeData (Variables|Default|GetHeatEventFireModeData))
      (Rendering|Material|SetScalarParameterValue (Variables|Default|GetMatWeaponGlow) "PowerUp_Heat" 1.0 :!class /Script/Engine.MaterialInstanceDynamic)
      (Rendering|Material|SetScalarParameterValue (Variables|Default|GetMatWeaponGlow) "PowerUp_HeatTime" (Utilities|Time|GetGameTimeinSeconds) :!class /Script/Engine.MaterialInstanceDynamic)
      (Audio|Components|Audio|FadeIn (Variables|Default|GetPowerupLoop) 1.0 1.0 0.0 "Linear" :!class /Script/Engine.AudioComponent))))

(event Custom|OnStartHeatEvent
  ;[1 1 1] Heat Event Logic
  ;;[1 1 1] Server Only. Handles heat event and reps bool to clients
  (if (not (Variables|Default|GetbIsInHeatEvent))
    (Variables|Default|SetbIsInHeatEvent true)
    (Networking|ForceNetUpdate)))

(event Custom|OnEndHeatEvent
  ;[1 1 1] Heat Event Logic
  ;;[1 1 1] Server Only. Handles heat event and reps bool to clients
  (if (Variables|Default|GetbIsInHeatEvent)
    (Variables|Default|SetbIsInHeatEvent false)
    (Networking|ForceNetUpdate)))

(event Custom|OnEndHeatEvent_Rep
  ;[1 1 1] Heat Event Logic
  ;;[1 1 1] OnEndHeatEvent
  (Networking|CanExecuteCosmeticEvents
    (:True
      (FortWeapon|ClearOverrideFireModeData)
      (Audio|Components|Audio|FadeOut (Variables|Default|GetPowerupLoop) 0.3 0.0 "Linear" :!class /Script/Engine.AudioComponent)
      (if (not (Overheat|IsOverheated))
        ;;;[1 1 1] On End Heat Event. Naturally cooling down.
        (Rendering|Material|SetScalarParameterValue (Variables|Default|GetMatWeaponGlow) "PowerUp_Heat" 0.0 :!class /Script/Engine.MaterialInstanceDynamic)
        (Rendering|Material|SetScalarParameterValue (Variables|Default|GetMatWeaponGlow) "PowerUp_HeatTime" (Utilities|Time|GetGameTimeinSeconds) :!class /Script/Engine.MaterialInstanceDynamic)))))

(event Overheating|EventOnStartOverheated
  (bind _matweaponglow (Variables|Default|GetMatWeaponGlow))
  ;[1 1 1] OnOverheated
  (|CallParentFunction)
  (CallFunction|OnEndHeatEvent)
  (Rendering|Material|SetScalarParameterValue _matweaponglow "PowerUp_Cooldown" 1.0 :!class /Script/Engine.MaterialInstanceDynamic)
  (Rendering|Material|SetScalarParameterValue _matweaponglow "PowerUp_CooldownTime" (Utilities|Time|GetGameTimeinSeconds) :!class /Script/Engine.MaterialInstanceDynamic)
  (Variables|Default|SetbIsInCooldown true)
  (Components|Activation|Activate (Variables|Default|GetNS_Powerup_Rifle_Overheat) true)
  (Audio|Components|Audio|Play (Variables|Default|GetOverheatAudioLayerComp) 0.0))

(event Overheating|EventOnFinishedCooling
  ;[1 1 1] On Completely Cooled off. No Heat.
  (|CallParentFunction)
  (Networking|CanExecuteCosmeticEvents
    (:True
      (if (Variables|Default|GetbIsInCooldown)
        (Rendering|Material|SetScalarParameterValue (Variables|Default|GetMatWeaponGlow) "PowerUp_Cooldown" 0.0 :!class /Script/Engine.MaterialInstanceDynamic)
        (Rendering|Material|SetScalarParameterValue (Variables|Default|GetMatWeaponGlow) "PowerUp_CooldownTime" (Utilities|Time|GetGameTimeinSeconds) :!class /Script/Engine.MaterialInstanceDynamic)
        (Rendering|Material|SetScalarParameterValue (Variables|Default|GetMatWeaponGlow) "PowerUp_Heat" 0.0 :!class /Script/Engine.MaterialInstanceDynamic)
        (Rendering|Material|SetScalarParameterValue (Variables|Default|GetMatWeaponGlow) "PowerUp_HeatTime" (Utilities|Time|GetGameTimeinSeconds) :!class /Script/Engine.MaterialInstanceDynamic)
        (Variables|Default|SetbIsInCooldown false)))))

(event EventOnPlayWeaponFireFX
  (params (bPersistentFire :type Boolean) (bSecondaryFire :type Boolean))
  ;[1 1 1] Material Glow FX
  (|CallParentFunction bPersistentFire false)
  (Rendering|Material|SetScalarParameterValue (Variables|Default|GetMatWeaponGlow) "PowerUp_FireTime" (Utilities|Time|GetGameTimeinSeconds) :!class /Script/Engine.MaterialInstanceDynamic)
  (bind _returnvalue (Effects|SpawnFXSystemAttached :FXSystemAsset "/CorruptionGameplay/Gameplay/Items/Weapons/PowerUp/Effects/NS_Powerup_Rifle_MF.NS_Powerup_Rifle_MF" :AttachToComponent (Variables|Visual|GetWeaponMesh) :AttachPointName "Muzzle" :Scale "1.000000,1.000000,1.000000" :LocationType "SnapToTarget" :bAutoDestroy true :PoolingMethod "None" :bAutoActivate true))
  (Effects|Components|ParticleSystem|SetFloatParameter _returnvalue "User.Overheat" (Math|Float|MapRangeClamped (Variables|Default|GetOverheatSmoothed) 0.08 1.0 0.0 1.0))
  (bind _overheatshotlayer (Variables|Default|GetOverheatShotLayer)
    (:then
      (bind _returnvalue_2 (Audio|SpawnSoundAttached :Sound _overheatshotlayer :AttachToComponent (Variables|Default|GetOverheatAudioLayerComp) :AttachPointName "None" :LocationType "KeepRelativeOffset" :bStopWhenAttachedToDestroyed true :VolumeMultiplier 1.0 :PitchMultiplier 1.0 :StartTime 0.0 :bAutoDestroy true))
      (bind _output_get (Variables|Default|SetOverheatShotLayerComp _returnvalue_2))
      (Audio|Parameter|SetFloatParameter _output_get "Overheat" (CallFunction|GetCurrentHeatPercent) :!class /Script/AudioExtensions.AudioParameterControllerInterface))
    (:else)))

(event Custom|BarrelOverheatFXandSound
  ;[1 1 1] Material Glow FX
  (Networking|CanExecuteCosmeticEvents
    (:True
      (Rendering|Material|SetScalarParameterValue (Variables|Default|GetMatWeaponGlow) "PowerUp_HeatAmount" (CallFunction|GetCurrentHeatPercent) :!class /Script/Engine.MaterialInstanceDynamic)
      (Variables|Default|SetOverheatSmoothed (Math|Interpolation|FInterpTo (Variables|Default|GetOverheatSmoothed) (Math|Float|Lerp 0.0 1.0 (CallFunction|GetCurrentHeatPercent)) (Utilities|Time|GetWorldDeltaSeconds) 1e+01))
      (Effects|Components|ParticleSystem|SetFloatParameter (Variables|Default|GetNS_Powerup_Rifle_Overheat) "User.Overheat" (Math|Float|MapRangeClamped (Variables|Default|GetOverheatSmoothed) 0.08 1.0 0.0 1.0))
      (Audio|Parameter|SetFloatParameter (Variables|Default|GetPowerupLoop) "Overheat" (CallFunction|GetCurrentHeatPercent) :!class /Script/AudioExtensions.AudioParameterControllerInterface))))

(fn GetCurrentHeatPercent
  (return (Math|Float|SafeDivide (Variables|Default|GetCurrentHeatValueRep) (Overheat|GetOverheatingMaxValue))))

(fn OnRep_bIsInHeatEvent
  (if (Variables|Default|GetbIsInHeatEvent)
    (CallFunction|OnStartHeatEventRep)
    (else
      (CallFunction|OnEndHeatEventRep))))

(fn OnRep_CurrentHeatValueRep
  (CallFunction|BarrelOverheatFXandSound))

(fn ConstructionScript
  (|CallParentFunction)
  (bind _returnvalue (Rendering|Material|CreateDynamicMaterialInstance (Variables|Visual|GetWeaponMesh) 0 "/CorruptionGameplay/Gameplay/Items/Weapons/PowerUp/Rifle/Materials/MI_Power_Up_Rifle.MI_Power_Up_Rifle" "None" :!class /Script/Engine.PrimitiveComponent))
  (Variables|Default|SetMatWeaponGlow _returnvalue))
